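# Niagara system for the Classic vehicle drift trail cosmetic, drawing two skeletal-attached ribbons masked per tire.
# Game mode selects spawn offset, lifetime, width and backward velocity, while painted-variant overrides set primary and secondary colors.
# Two ribbon renderers share each particle, with separate back color and width, curve-driven fades, inherited source movement and user-driven material gradient and opacity.
# A second, near-duplicate emitter with longer lifetime is disabled.
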NiagaraSystem /VehicleCosmetics/DriftTrail/DriftTrail_Classic/FX/NS_DriftTrail_Classic_MP
dependencies {
 usedDataInterfaces [&NiagaraDataInterfaceArrayFloat3, &NiagaraDataInterfaceCurve, &NiagaraDataInterfaceSkeletalMesh]
 usedDynamicInputs []
 usedModules [
  &self:SystemSpawnScript.SystemScriptSource.SystemScriptGraph.NiagaraNodeAssignment_0.SetVariables_#1
  &self:SystemSpawnScript.SystemScriptSource.SystemScriptGraph.NiagaraNodeAssignment_1.SetVariables_#2
  &FN_UnpackVehicleFlags
  /VehicleCosmetics/FX/Niagara/Modules/VC_SetVehicleSpeedAttribs
  &SystemState
 ]
}
emitters {
 NE_VC_DriftTrail_Ribbon_Parent {
  emitterUpdateScript [
   &EmitterState()
   &VC_RateQualityHandler(Epic Spawn Rate=30, High Spawn Rate=30, Low Spawn Rate=30, Use Custom Rate?=true)
   &VC_SpawnRibbon_v1(Number of Ribbons=2, SpawnRate=$Emitter.Spawn Rate)
  ]
  particleSpawnScript [
   &VC_PaintedVariantsOverride(
    Black=<rgba 0.007 0.007 0.007 50>
    BurntSienna=<rgba 0.055 0.0316 0.022 20>
    Cobalt=<rgba 0.021 0.1657 0.6 1>
    DefaultColor=<rgba 0.5452 0.0307 1 1>
    ForestGreen=<rgba 0 0.35 0.0177 1>
    Lime=<rgba 0.3924 0.5885 0 1>
    Orange=<rgba 0.65 0.39 0.13 1>
    Pink=<rgba 0.45 0.135 0.4237 1>
    Purple=<rgba 0.4973 0.039 0.65 0.9>
    R1 Black:LinearColor=<rgba 0.007 0.007 0.007 20>
    R1 Cobalt:LinearColor=<rgba 0 0.25 1 1>
    R1 DefaultColor=<rgba 0 0.2125 0.85 5>
    R1 ForestGreen:LinearColor=<rgba 0 0.25 0.0126 1>
    R1 Orange:LinearColor=<rgba 0.6 0.36 0.12 0>
    R1 Pink:LinearColor=<rgba 0.56 0.0671 0.3135 0>
    R1 Saffron:LinearColor=<rgba 0.9 0.69 0 0>
    R1 TitaniumWhite:LinearColor=<rgba 0.25 0.25 0.25 0>
    Saffron=<rgba 0.7 0.6417 0 0.95>
    SkyBlue=<rgba 0 0.499 0.75 1>
    Use R1 Painted Overrides=true
   )
   &VC_InitializeRibbon_v1(
    Color=$Particles.Color
    Custom Offset=true
    Enable Skeletal Position=true
    Lifetime=/VehicleCosmetics/FX/Niagara/Modules/VC_Mask_Floats(Coro Value:NiagaraFloat=0.4, FrontEnd Value:NiagaraFloat=0.5, FN Value:NiagaraFloat=0.5)
    Local Vector Offset=&SelectVectorFromArray(
     Direct Array Index=$System.GameModeID
     Vector Selection Array=DataInterface() {propertyValues {FloatData [<-28 0 -20>, <-28 0 -20>, <-28 0 -20>, <-28 0 -40>, <-28 0 -40>]}}
    )
    Offset Amount=1
    Ribbon Width:NiagaraFloat=&VC_Mask_Floats(Coro Value:NiagaraFloat=10, FrontEnd Value:NiagaraFloat=13.5, FN Value:NiagaraFloat=13.5)
   )
   &AddVelocity(Velocity=&VC_Mask_Vectors(Coro Value:Vector3f=<-800 0 0>, FrontEnd Value:Vector3f=<-1300 0 0>, FN Value:Vector3f=<-1000 0 0>)) {module=AddVelocity001}
   /VehicleCosmetics/FX/VC_FXParents/Modules/DLR_ProductPreviews(GameModeID:NiagaraInt32=$System.GameModeID, bPreviewVehicleSpeed:NiagaraBool=$System.bPreviewVehicleSpeed)
  ]
  particleUpdateScript [
   &FN_DisableTireFX(Mask Index Attribute:NiagaraInt32=$Particles.SpawnGroup, Mask Index FL=2, Mask Index FR=3, Mask Index RL=0, Mask Index RR=1)
   &ParticleState()
   &InheritSourceMovement(
    Applied Position Scale=&Multiply_VectorByFloat(
     Vector=<1 1 1>
     Float=&FloatFromCurve(FloatCurve=DataInterface() {propertyValues {Curve {keys [
          {interpMode=RCIM_Cubic, leaveTangent=-0.0051, tangentMode=RCTM_Break, time=0, value=1}
          {arriveTangent=-0.247, interpMode=RCIM_Cubic, tangentMode=RCTM_Break, time=1, value=0.7}
         ]}}})
    )
    ApplyForce=false
    ApplyPosition=true
    ApplyPositionFalloff=false
   ) {enabled=false, module=InheritSourceMovement001}
   &ScaleRibbonWidth(Ribbon Width Scale=&FloatFromCurve(
     FloatCurve=DataInterface() {propertyValues {Curve {keys [
         {arriveTangent=0.0434, interpMode=RCIM_Cubic, leaveTangent=1.6245, tangentMode=RCTM_Break, time=0, value=0.85}
         {arriveTangent=0.0114, interpMode=RCIM_Cubic, leaveTangent=0.0114, tangentMode=RCTM_User, time=0.125, value=1}
         {arriveTangent=-0.5443, interpMode=RCIM_Cubic, leaveTangent=-0.5443, tangentMode=RCTM_User, time=1, value=0.5}
        ]}}}
     Scale Curve=4
    ))
   &ScaleColor(
    Scale Alpha=&FloatFromCurve(FloatCurve=DataInterface() {propertyValues {Curve {keys [
         {tangentMode=RCTM_User, time=0, value=1}
         {arriveTangent=-0.0405, interpMode=RCIM_Cubic, leaveTangent=-0.0405, tangentMode=RCTM_User, time=0.7009, value=1}
         {arriveTangent=-8.0089, interpMode=RCIM_Cubic, leaveTangent=-8.0089, tangentMode=RCTM_User, time=1, value=0}
        ]}}})
    Scale RGB=&VectorFromFloat(Value=&FloatFromCurve(FloatCurve=DataInterface() {propertyValues {Curve {keys [{interpMode=RCIM_Cubic, tangentMode=RCTM_User, time=0.3, value=10}, {interpMode=RCIM_Cubic, tangentMode=RCTM_User, time=1, value=1.5}]}}}))
    ScaleA=true
    ScaleRGB=true
   ) {module=ScaleColor001}
   &ScaleColor(
    Scale Alpha=&FloatFromCurve(FloatCurve=DataInterface() {propertyValues {Curve {keys [
         <at 0 1>
         {interpMode=RCIM_Cubic, tangentMode=RCTM_Break, time=0.8, value=1}
         {interpMode=RCIM_Cubic, tangentMode=RCTM_Break, time=1, value=0}
        ]}}})
    Scale RGB=&VectorFromFloat(Value=$User.PaintedScalar)
    ScaleA=true
    ScaleRGB=true
   ) {module=ScaleColor002}
   /SetVariables(
    Particles.BackColor:LinearColor=&Multiply_LinearColorByFloat(Linear Color=$Particles.R1 Color, Float=1)
    Particles.BackRibbonWidth:NiagaraFloat=&FloatFromCurve(
     FloatCurve=DataInterface() {propertyValues {Curve {keys [
         {arriveTangent=0.0434, interpMode=RCIM_Cubic, leaveTangent=1.6245, tangentMode=RCTM_Break, time=0, value=0.85}
         {arriveTangent=0.0114, interpMode=RCIM_Cubic, leaveTangent=0.0114, tangentMode=RCTM_User, time=0.125, value=1}
         {arriveTangent=-0.5443, interpMode=RCIM_Cubic, leaveTangent=-0.5443, tangentMode=RCTM_User, time=1, value=0.5}
        ]}}}
     Scale Curve=70
    )
   ) {isSetParameters=true, module=SetVariables_#3}
   &DynamicMaterialParameters(
    Index 0 Param 2=$User.Opacity
    Index 1 Param 1=$User.gradient_r
    Index 1 Param 2=$User.gradient_g
    Index 1 Param 3=$User.gradient_b
    Module.Index0Param0Write=true
    Module.Index0Param1Write=true
    Module.Index0Param2Write=true
    Module.Index0Param3Write=true
    Module.Index1Param0Write=true
    Module.Index1Param1Write=true
    Module.Index1Param2Write=true
    Module.Index1Param3Write=true
    Write Parameter Index 1=true
   )
   &InheritSourceMovement(
    Applied Position Falloff Distance=650
    Applied Position Scale=&VectorFromFloat(Value=&FloatFromCurve(FloatCurve=DataInterface() {propertyValues {Curve {
         defaultValue=1
         keys [
          {arriveTangent=-1.7885, interpMode=RCIM_Cubic, leaveTangent=-1.7885, tangentMode=RCTM_User, time=0.05, value=1}
          {interpMode=RCIM_Cubic, tangentMode=RCTM_User, time=1, value=0}
         ]
        }}}))
    ApplyForce=false
    ApplyPosition=true
    ApplyPositionFalloff=false
   )
   &SolveForcesAndVelocity(Clamp Velocity=false, Limit Acceleration=false)
   &VC_UpdateRibbon_v1(Enable Skeletal Position=true)
  ]
  rendererClasses [&NiagaraRibbonRendererProperties]
  renderers [
   {rendererClass=&NiagaraRibbonRendererProperties}
   {rendererClass=&NiagaraRibbonRendererProperties, rendererIndex=1}
  ]
 }
 NE_VC_DriftTrail_Ribbon_Parent001 {
  bEnabled=false
  emitterUpdateScript [
   &EmitterState()
   &VC_RateQualityHandler(Epic Spawn Rate=30, High Spawn Rate=30, Low Spawn Rate=30, Use Custom Rate?=true)
   &VC_SpawnRibbon_v1(Number of Ribbons=2, SpawnRate=$Emitter.Spawn Rate)
  ]
  particleSpawnScript [
   &VC_PaintedVariantsOverride(
    Black=<rgba 0.007 0.007 0.007 50>
    BurntSienna=<rgba 0.055 0.0316 0.022 20>
    Cobalt=<rgba 0.021 0.1657 0.6 1>
    DefaultColor=<rgba 0.5452 0.0307 1 1>
    ForestGreen=<rgba 0 0.35 0.0177 1>
    Lime=<rgba 0.3924 0.5885 0 1>
    Orange=<rgba 0.65 0.39 0.13 1>
    Pink=<rgba 0.45 0.135 0.4237 1>
    Purple=<rgba 0.4973 0.039 0.65 0.9>
    R1 Black:LinearColor=<rgba 0.007 0.007 0.007 20>
    R1 Cobalt:LinearColor=<rgba 0 0.25 1 1>
    R1 DefaultColor=<rgba 0 0.2125 0.85 5>
    R1 ForestGreen:LinearColor=<rgba 0 0.25 0.0126 1>
    R1 Orange:LinearColor=<rgba 0.6 0.36 0.12 0>
    R1 Pink:LinearColor=<rgba 0.56 0.0671 0.3135 0>
    R1 Saffron:LinearColor=<rgba 0.9 0.69 0 0>
    R1 TitaniumWhite:LinearColor=<rgba 0.25 0.25 0.25 0>
    Saffron=<rgba 0.7 0.6417 0 0.95>
    SkyBlue=<rgba 0 0.499 0.75 1>
    Use R1 Painted Overrides=true
   )
   &VC_InitializeRibbon_v1(
    Color=$Particles.Color
    Custom Offset=true
    Enable Skeletal Position=true
    Lifetime=0.8
    Local Vector Offset=&SelectVectorFromArray(
     Direct Array Index=$System.GameModeID
     Vector Selection Array=DataInterface() {propertyValues {FloatData [<-28 0 -20>, <-28 0 -20>, <-28 0 -20>, <-28 0 -40>, <-28 0 -40>]}}
    )
    Offset Amount=1
    Ribbon Width:NiagaraFloat=&VC_Mask_Floats(Coro Value:NiagaraFloat=10, FrontEnd Value:NiagaraFloat=13.5, FN Value:NiagaraFloat=13.5)
   )
   &AddVelocity(Velocity=&VC_Mask_Vectors(Coro Value:Vector3f=<-800 0 0>, FrontEnd Value:Vector3f=<-1300 0 0>, FN Value:Vector3f=<-1000 0 0>)) {module=AddVelocity001}
  ]
  particleUpdateScript [
   &FN_DisableTireFX(Mask Index Attribute:NiagaraInt32=$Particles.SpawnGroup, Mask Index FL=2, Mask Index FR=3, Mask Index RL=0, Mask Index RR=1)
   &ParticleState()
   &InheritSourceMovement(
    Applied Position Scale=&Multiply_VectorByFloat(
     Vector=<1 1 1>
     Float=&FloatFromCurve(FloatCurve=DataInterface() {propertyValues {Curve {keys [
          {interpMode=RCIM_Cubic, leaveTangent=-0.0051, tangentMode=RCTM_Break, time=0, value=1}
          {arriveTangent=-0.247, interpMode=RCIM_Cubic, tangentMode=RCTM_Break, time=1, value=0.7}
         ]}}})
    )
    ApplyForce=false
    ApplyPosition=true
    ApplyPositionFalloff=false
   ) {module=InheritSourceMovement001}
   &ScaleRibbonWidth(Ribbon Width Scale=&FloatFromCurve(
     FloatCurve=DataInterface() {propertyValues {Curve {keys [
         {arriveTangent=0.0434, interpMode=RCIM_Cubic, leaveTangent=1.6245, tangentMode=RCTM_Break, time=0, value=0.85}
         {arriveTangent=0.0114, interpMode=RCIM_Cubic, leaveTangent=0.0114, tangentMode=RCTM_User, time=0.125, value=1}
         {arriveTangent=-0.5443, interpMode=RCIM_Cubic, leaveTangent=-0.5443, tangentMode=RCTM_User, time=1, value=0.5}
        ]}}}
     Scale Curve=4
    ))
   &ScaleColor(
    Scale Alpha=&FloatFromCurve(FloatCurve=DataInterface() {propertyValues {Curve {keys [
         {tangentMode=RCTM_User, time=0, value=1}
         {arriveTangent=-0.0405, interpMode=RCIM_Cubic, leaveTangent=-0.0405, tangentMode=RCTM_User, time=0.7009, value=1}
         {arriveTangent=-8.0089, interpMode=RCIM_Cubic, leaveTangent=-8.0089, tangentMode=RCTM_User, time=1, value=0}
        ]}}})
    Scale RGB=&VectorFromFloat(Value=&FloatFromCurve(FloatCurve=DataInterface() {propertyValues {Curve {keys [{interpMode=RCIM_Cubic, tangentMode=RCTM_User, time=0.3, value=10}, {interpMode=RCIM_Cubic, tangentMode=RCTM_User, time=1, value=1.5}]}}}))
    ScaleA=true
    ScaleRGB=true
   ) {module=ScaleColor001}
   &ScaleColor(
    Scale Alpha=&FloatFromCurve(FloatCurve=DataInterface() {propertyValues {Curve {keys [
         <at 0 1>
         {interpMode=RCIM_Cubic, tangentMode=RCTM_Break, time=0.8, value=1}
         {interpMode=RCIM_Cubic, tangentMode=RCTM_Break, time=1, value=0}
        ]}}})
    Scale RGB=&VectorFromFloat(Value=$User.PaintedScalar)
    ScaleA=true
    ScaleRGB=true
   ) {module=ScaleColor002}
   /SetVariables(
    Particles.BackColor:LinearColor=&Multiply_LinearColorByFloat(Linear Color=$Particles.R1 Color, Float=1)
    Particles.BackRibbonWidth:NiagaraFloat=&FloatFromCurve(
     FloatCurve=DataInterface() {propertyValues {Curve {keys [
         {arriveTangent=0.0434, interpMode=RCIM_Cubic, leaveTangent=1.6245, tangentMode=RCTM_Break, time=0, value=0.85}
         {arriveTangent=0.0114, interpMode=RCIM_Cubic, leaveTangent=0.0114, tangentMode=RCTM_User, time=0.125, value=1}
         {arriveTangent=-0.5443, interpMode=RCIM_Cubic, leaveTangent=-0.5443, tangentMode=RCTM_User, time=1, value=0.5}
        ]}}}
     Scale Curve=70
    )
   ) {isSetParameters=true, module=SetVariables_#3}
   &DynamicMaterialParameters(
    Index 0 Param 2=$User.Opacity
    Index 1 Param 1=$User.gradient_r
    Index 1 Param 2=$User.gradient_g
    Index 1 Param 3=$User.gradient_b
    Module.Index0Param0Write=true
    Module.Index0Param1Write=true
    Module.Index0Param2Write=true
    Module.Index0Param3Write=true
    Module.Index1Param0Write=true
    Module.Index1Param1Write=true
    Module.Index1Param2Write=true
    Module.Index1Param3Write=true
    Write Parameter Index 1=true
   )
   &SolveForcesAndVelocity(Clamp Velocity=false, Limit Acceleration=false)
   &VC_UpdateRibbon_v1(Enable Skeletal Position=true)
  ]
  rendererClasses [&NiagaraRibbonRendererProperties]
  renderers [
   {rendererClass=&NiagaraRibbonRendererProperties}
   {rendererClass=&NiagaraRibbonRendererProperties, rendererIndex=1}
  ]
 }
}
summary {
 systemName=NS_DriftTrail_Classic_MP
 emitters [
  {emitterName=NE_VC_DriftTrail_Ribbon_Parent, rendererClasses [&NiagaraRibbonRendererProperties]}
  {bEnabled=false, emitterName=NE_VC_DriftTrail_Ribbon_Parent001, rendererClasses [&NiagaraRibbonRendererProperties]}
 ]
 userVariables [
  User.bIsFrontendPreview {type=NiagaraBool, defaultValue=true}
  User.bIsPainted {type=NiagaraBool}
  User.bPreviewVehicleSpeed {type=NiagaraBool}
  User.ColorVariantID {type=NiagaraInt32}
  User.GameModeID {type=NiagaraInt32, defaultValue=3}
  User.gradient_b {type=NiagaraFloat, defaultValue=100, description=""}
  User.gradient_g {type=NiagaraFloat, defaultValue=0, description=""}
  User.gradient_r {type=NiagaraFloat, defaultValue=0, description=""}
  User.offset {type=Vector3f, description="offset from wheels(set through cosmetic template)", defaultValue {struct=Vector3f, value=<0 0 0>}}
  User.Opacity {type=NiagaraFloat, defaultValue=1, description=""}
  User.PaintedColor {type=LinearColor, description="", defaultValue {struct=LinearColor, value=<rgba 0.5452 0.0307 1 1>}}
  User.PaintedScalar {type=NiagaraFloat, defaultValue=1, description=""}
  User.VehicleFlags {type=NiagaraInt32, defaultValue=806289408}
  User.Wheel_Radius_Array {
   type=NiagaraDataInterfaceArrayFloat
   defaultValue {
    struct=NiagaraExt_VariableValue_DataInterface
    value {
     dataInterface=&self:NiagaraDataInterfaceArrayFloat_1
     dataInterfaceClass=&NiagaraDataInterfaceArrayFloat
    }
   }
  }
  User.Wheel_Width_Array {
   type=NiagaraDataInterfaceArrayFloat
   defaultValue {
    struct=NiagaraExt_VariableValue_DataInterface
    value {
     dataInterface=&self:NiagaraDataInterfaceArrayFloat_0
     dataInterfaceClass=&NiagaraDataInterfaceArrayFloat
    }
   }
  }
  User.WheelTouchingArray {
   type=NiagaraDataInterfaceArrayFloat
   defaultValue {
    struct=NiagaraExt_VariableValue_DataInterface
    value {
     dataInterface=&self:NiagaraDataInterfaceArrayFloat_2
     dataInterfaceClass=&NiagaraDataInterfaceArrayFloat
    }
   }
  }
 ]
}
